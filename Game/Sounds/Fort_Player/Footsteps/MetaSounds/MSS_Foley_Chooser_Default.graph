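# Default one-shot mono MetaSound source for player footstep foley, playing whichever wave asset the Fort.FoleyAction interface supplies.
# On play, two chained random float nodes roll pitch and gain before triggering the wave player.
# Gain varies slightly around unity, but the same range feeds the semitone-based Pitch Shift, so pitch sits about one semitone up.
# The Fort.Character inputs are exposed but unused, and playback completion fires OnFinished.
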MetaSoundSource /Game/Sounds/Fort_Player/Footsteps/MetaSounds/MSS_Foley_Chooser_Default
inputs [
 Character.Affiliation {type=Enum:FortAudioAffiliation}
 Character.Health {type=Float}
 Character.UniqueID {type=Int32}
 FoleyAction.WaveAsset {type=WaveAsset}
 UE.Source.OnPlay {type=Trigger}
]
interfaces [Fort.Character:1.0, Fort.FoleyAction:1.0, UE.OutputFormat.Mono:1.0, UE.Source.OneShot:1.0, UE.Source:1.1]
outputs [
 "UE.OutputFormat.Mono.Audio:0" {type=Audio}
 UE.Source.OneShot.OnFinished {type=Trigger}
]
pages [Default {
  input_defaults {Character.Affiliation=0, Character.Health=0, Character.UniqueID=0, FoleyAction.WaveAsset=null, UE.Source.OnPlay=null}
  nodes [
   n1 = UE.RandomFloat(Next=^UE.Source.OnPlay) {also {defaults {Max=1.05, Min=0.95}}}
   n2 = UE.RandomFloat(Next=@n1.On Next) {also {defaults {Max=1.05, Min=0.95}}}
   WavePlayer_0 = UE.Wave Player:Mono(Wave Asset=^FoleyAction.WaveAsset, Pitch Shift=@n1.Value, Play=@n2.On Next) {also {version [1,1]}}
   Multiply_0 = UE.Multiply:Audio by Float(AdditionalOperands=@n2.Value, PrimaryOperand=@WavePlayer_0.Out Mono)
  ]
  links [{dst="@UE.OutputFormat.Mono.Audio:0", src=Multiply_0.Out}, {dst="@UE.Source.OneShot.OnFinished", src=WavePlayer_0.On Finished}]
 }]
properties {assetImportData=&self:AssetImportData, sampleRate=48000}
