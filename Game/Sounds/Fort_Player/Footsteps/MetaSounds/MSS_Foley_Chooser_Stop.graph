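# MSS_Foley_Chooser_Stop is a mono one-shot MetaSound Source for player footstep foley, named for the stop action.
# It is a preset of MSS_Foley_Chooser_Default, which supplies the graph logic.
# Fort.Character and Fort.FoleyAction interfaces expose affiliation, health, unique ID and a wave asset array as inputs.
# All inputs stay at zero or null defaults, with no wave assets assigned here.
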MetaSoundSource /Game/Sounds/Fort_Player/Footsteps/MetaSounds/MSS_Foley_Chooser_Stop
preset_parent=&/MSS_Foley_Chooser_Default
inputs [
 Character.Affiliation {type=Enum:FortAudioAffiliation}
 Character.Health {type=Float}
 Character.UniqueID {type=Int32}
 FoleyAction.WaveAssets {type=WaveAsset:Array}
 UE.Source.OnPlay {type=Trigger}
]
interfaces [Fort.Character:1.0, Fort.FoleyAction:1.0, UE.OutputFormat.Mono:1.0, UE.Source.OneShot:1.0, UE.Source:1.1]
outputs [
 "UE.OutputFormat.Mono.Audio:0" {type=Audio}
 UE.Source.OneShot.OnFinished {type=Trigger}
]
pages [Default {input_defaults {Character.Affiliation=0, Character.Health=0, Character.UniqueID=0, FoleyAction.WaveAssets=null, UE.Source.OnPlay=null}}]
properties {assetImportData=&self:AssetImportData}
